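# Sound cue for the third-person canting foley of the SMG, heard by players near another player's weapon.
# A random node picks between two equally weighted wave recordings and feeds a modulator that holds volume at a fixed reduced level.
# The cue adds a volume boost and uses the weapon reload attenuation and the weapon foley sound class.
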SoundCue /Game/Sounds/Fort_Weapons/Canting/SMG/Weapon_SMG_Canting_ThirdPerson_Cue
root=Modulator_0
nodes [
 {id=Modulator_0, children [Random_0], properties {volumeMax=0.7, volumeMin=0.7}}
 {id=Random_0, children [WavePlayer_0,WavePlayer_3], properties {weights [1,1]}}
 {id=WavePlayer_0, properties {soundWaveAssetPtr=&/Weapon_SMG_Canting_ThirdPerson_01}}
 {id=WavePlayer_3, properties {soundWaveAssetPtr=&/Weapon_SMG_Canting_ThirdPerson_02}}
]
properties {
 attenuationSettings=&Fort_Weapon_Reload_Att
 maxDistance=2250
 soundClassObject=&Fort_WeaponFoley
 volumeMultiplier=1.5
}
